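# Unlit emissive material for the glowing star effect on the mint Bright Bomber backpack cosmetic, usable on Niagara mesh particles.
# A Fresnel term blends two bright HDR mint-green colors toward the edges, and DeriveHDRfromLDR remaps the clamped result with the unclamped color as tint.
# A shading path switch gives forward and mobile rendering just the flat base color.
# The result passes through MF_FX_EmissiveAdjustments before output.
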Material /BRCosmetics/Accessories/FORT_Backpacks/Backpack_F_MED_Bright_Bomber_Mint/FX/M_BBM_F_BackpackStarEmissive
graph {
 properties {bUsedWithNiagaraMeshParticles=true, pixelDepthOffsetMode=PDOM_Legacy, shadingModel=MSM_Unlit, usedShadingModels=Unlit}
 nodes [
  n1 = Constant3Vector() {constant=<rgba 0.1542 3 0.205 0>}
  n2 = LinearInterpolate(A=@n1, B=Constant3Vector() {constant=<rgba 0 6 0.656 0>}, Alpha=Fresnel() {exponent=1})
  n3 = Clamp(None=@n2)
  MaterialFunctionCall(Emissive=ShadingPathSwitch(
    Default=MaterialFunctionCall(LDR Input=@n3, HDR Tint=@n2) {materialFunction=&DeriveHDRfromLDR}.RemappedRange
    Forward=@n1
    Mobile=@n1
   )) {materialFunction=&MF_FX_EmissiveAdjustments}
 ]
}
